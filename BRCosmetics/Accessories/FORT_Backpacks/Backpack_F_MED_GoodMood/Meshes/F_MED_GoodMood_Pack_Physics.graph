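# Physics asset for the GoodMood backpack cosmetic mesh, defining simulated secondary motion for its handle and tassel.
# Small capsule bodies cover a base bone for each part, one dynamic handle bone and a five-link dynamic tassel chain.
# The handle hangs from its base with wider swing and locked twist.
# The tassel links connect in series from their base with tight swing and twist limits, with one swing axis loosening further down the chain.
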PhysicsAsset /BRCosmetics/Accessories/FORT_Backpacks/Backpack_F_MED_GoodMood/Meshes/F_MED_GoodMood_Pack_Physics
bodies {
 base_handle [{shapeType=Capsule, length=0.505, radius=0.505, shapeName=base_handle_capsule, rotation=<rot 0 -90 0>}]
 base_tassle [{shapeType=Capsule, length=0.505, radius=0.505, shapeName=base_tassle_capsule, rotation=<rot 0 -90 0>}]
 dyn_handle [{shapeType=Capsule, length=2, radius=1, shapeName=dyn_handle_capsule, center=<2 0 0>, rotation=<rot 90 0 0>}]
 dyn_tassle_00 [{shapeType=Capsule, length=2, radius=1, shapeName=dyn_tassle_00_capsule, center=<2 0 0>, rotation=<rot 90 0 0>}]
 dyn_tassle_01 [{shapeType=Capsule, length=2, radius=1, shapeName=dyn_tassle_01_capsule, center=<2 0 0>, rotation=<rot 90 0 0>}]
 dyn_tassle_02 [{shapeType=Capsule, length=2, radius=1, shapeName=dyn_tassle_02_capsule, center=<2 0 0>, rotation=<rot 90 0 0>}]
 dyn_tassle_03 [{shapeType=Capsule, length=2, radius=1, shapeName=dyn_tassle_03_capsule, center=<2 0 0>, rotation=<rot 90 0 0>}]
 dyn_tassle_04 [{shapeType=Capsule, length=2, radius=1, shapeName=dyn_tassle_04_capsule, center=<2 0 0>, rotation=<rot 90 0 0>}]
}
constraints [
 {bone1Name=dyn_tassle_04, bone2Name=dyn_tassle_03, swing1LimitDegrees=10, swing2LimitDegrees=20, twistLimitDegrees=1}
 {bone1Name=dyn_tassle_00, bone2Name=base_tassle, swing1LimitDegrees=10, swing2LimitDegrees=10, twistLimitDegrees=1}
 {bone1Name=dyn_tassle_01, bone2Name=dyn_tassle_00, swing1LimitDegrees=10, swing2LimitDegrees=10, twistLimitDegrees=1}
 {bone1Name=dyn_tassle_02, bone2Name=dyn_tassle_01, swing1LimitDegrees=10, swing2LimitDegrees=20, twistLimitDegrees=1}
 {bone1Name=dyn_tassle_03, bone2Name=dyn_tassle_02, swing1LimitDegrees=10, swing2LimitDegrees=20, twistLimitDegrees=1}
 {bone1Name=dyn_handle, bone2Name=base_handle, swing1LimitDegrees=30, swing2LimitDegrees=30, twistLimitDegrees=10, twistMotion=Locked}
]
